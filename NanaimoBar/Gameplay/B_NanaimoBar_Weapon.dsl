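; NanaimoBar chargeable projectile weapon Blueprint, handling aim, effects, audio, animation and removal when empty.
; Charge percent maps to stages that feed AnimBPs, fire montages and sounds, while full charge changes muzzle cue, aim offset and projectile gravity.
; Fire audio retriggers the active sound unless it differs or a watched MetaSound output flags it as expiring.
; The server sends remaining ammo as gameplay event magnitude, with timed and unequip fallbacks removing the empty item.
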(event EventBeginPlay
  (|CallParentFunction)
  (CallFunction|CacheVariables)
  (Variables|FortWeapon|SetOnGetAimRotOverride (EventDispatchers|CreateEvent :!event GetAimRotOverride))
  ; Adjusts Gravity on Full Charge
  (EventDispatchers|BindEventtoOnProjectileSpawned (delegate Custom|OnProjectileSpawned_Event)))

(event Custom|OnProjectileSpawned_Event
  (params (Projectile :type Fort%20Projectile%20Base%20Object%20Reference))
  ; Adjusts Gravity on Full Charge
  (if (>= (FortWeapon|GetChargePercent) 1.0)
    (Projectile|SetGravityScale Projectile (ScalableFloat|GetCachedScalableFloatValue (Variables|Default|GetFullChargeGravityScale)))))

(event EventOnPlayWeaponFireFX
  (params (bPersistentFire :type Boolean) (bSecondaryFire :type Boolean))
  (bind _returnvalue (FortWeapon|GetChargePercent))
  (bind _returnvalue_2 (Math|Float|Floor _returnvalue))
  (bind _returnvalue_3 (Math|Conversions|ToFloat(Integer) _returnvalue_2))
  (bind _returnvalue_4 (Ability|GameplayCue|MakeGameplayCueParameters 0.0 _returnvalue_3))
  (bind _weaponmesh (Variables|Visual|GetWeaponMesh))
  (bind _returnvalue_5 (Components|GetOwner _weaponmesh))
  ;[1 1 1] Client weapon fired
  (|CallParentFunction bPersistentFire bSecondaryFire)
  (if (== _returnvalue_2 0)
    (GameplayCue|ExecuteGameplayCueOnActor(Burst) _returnvalue_5 (Variables|Default|GetGameplay%20Cue%20MuzzleFX) _returnvalue_4)
    (else
      (GameplayCue|ExecuteGameplayCueOnActor(Burst) _returnvalue_5 (Variables|Default|GetGameplay%20Cue%20MuzzleChargedFX) _returnvalue_4))))

(event EventMuzzleFlashFX
  (params (Persistent%20Fire :type Boolean)))

(event EventTick
  (params (DeltaSeconds :type Float%20%28single-precision%29))
  ;[0.6458 0.2625 0.01306 0.5] Get Charge Stage to inform AnimBPs
  (Utilities|FlowControl|Gate true
    (:Exit
      (CallFunction|GetChargeStage))))

(event EventOnStartCharge
  ;[0.6458 0.2625 0.01306 0.5] Get Charge Stage to inform AnimBPs
  (Utilities|FlowControl|Sequence
    (:then_0
      (Variables|Default|SetChargeStage 0)
      (Utilities|FlowControl|Gate true
        (:Exit
          (CallFunction|GetChargeStage))))
    (:then_1
      (Variables|Default|SetOnChargeStart true)
      (Utilities|FlowControl|Delay 0.2)
      (Variables|Default|SetOnChargeStart false))
    (:then_2
      (Components|Activation|Activate (Variables|Default|GetNS_NanaimoBar_Charge_FilmA) true))))

(event EventOnEndCharge
  ;[0.6458 0.2625 0.01306 0.5] Get Charge Stage to inform AnimBPs
  (Utilities|FlowControl|Sequence
    (:then_0
      (Utilities|FlowControl|Gate true
        (:Exit
          (CallFunction|GetChargeStage))))
    (:then_1
      (Variables|Default|SetOnChargeEnd true)
      (Utilities|FlowControl|Delay 0.2)
      (Variables|Default|SetOnChargeEnd false))
    (:then_2
      (Utilities|FlowControl|Delay 0.1)
      (Variables|Default|SetChargeStage 0))
    (:then_3
      (if (and (FortWeapon|IsEquipped) (Rendering|IsVisible (Variables|Visual|GetWeaponMesh)))
        (Animation|Montage|PlayMontage (Variables|/Script/Engine.Character|GetMesh (Variables|CachedVariables|GetPlayerPawn)) (Utilities|Select :Index (Variables|Default|GetChargeStage) :Option%200 "/NanaimoBar/Assets/Animation/Animation/NanaimoBar_Fire_NonTargeted_CMM_M.NanaimoBar_Fire_NonTargeted_CMM_M" :Option%201 "/NanaimoBar/Assets/Animation/Animation/NanaimoBar_Fire_NonTargeted_CMM_M.NanaimoBar_Fire_NonTargeted_CMM_M" :Option%202 "/NanaimoBar/Assets/Animation/Animation/NanaimoBar_Fire_Stage02_CMM_M.NanaimoBar_Fire_Stage02_CMM_M" :Option%203 "/NanaimoBar/Assets/Animation/Animation/NanaimoBar_Fire_Stage03_CMM_M.NanaimoBar_Fire_Stage03_CMM_M"))))
    (:then_4
      (Components|Activation|Deactivate (Variables|Default|GetNS_NanaimoBar_Charge_FilmA)))))

(event FortWeapon|EventOnEquippedAndAttached
  ;[0.5947 1 0.825] On ammo changed event to send gameplay event to ability
  (|CallParentFunction)
  (EventDispatchers|BindEventtoOnConsumeAmmo (EventDispatchers|CreateEvent :!event Custom|AmmoCountChangedGarbage)))

(event FortWeapon|EventOnUnEquippedAndDetached
  ;[0.5947 1 0.825] On ammo changed event to send gameplay event to ability
  (|CallParentFunction)
  (EventDispatchers|UnbindEventfromOnConsumeAmmo (EventDispatchers|CreateEvent :!event Custom|AmmoCountChangedGarbage))
  (bind _weapon_fire_audio (Variables|Default|GetWeapon%20Fire%20Audio)
    (:then
      (Audio|Components|Audio|FadeOut _weapon_fire_audio 0.2 0.0 "Linear" :!class /Script/Engine.AudioComponent))
    (:else)))

(event Custom|AmmoCountChangedGarbage
  (params (Quantity :type Integer))
  (bind _returnvalue (FortWeapon|GetRemainingAmmo))
  (bind _returnvalue_2 (FortWeapon|Ammo|HasInfiniteUses))
  (bind _fallbackremovaldelay (Variables|Default|GetFallbackRemovalDelay))
  (bind _returnvalue_3 (ScalableFloat|GetValueatLevel _fallbackremovaldelay))
  (bind _returnvalue_4 (select _returnvalue_2 1 _returnvalue))
  ;[0.5947 1 0.825] On ammo changed event to send gameplay event to ability
  (Networking|SwitchHasAuthority
    (:Authority
      (bind _asfort_player_pawn (Utilities|Casting|CastToFortPlayerPawn (Game|GetInstigator)))
      ;[0.01042 1 1] We check ammo in the GA this event triggers, but the client ammo count is not synced with server when we check to play montage and discard the weapon.
      ; Passing through ammo as event magnitude lets us properly dispose of the weapon
      (Ability|SendGameplayEventtoActor _asfort_player_pawn "(TagName=\"NanaimoBar.Item.AmmoChanged\")" (Utilities|Struct|MakeGameplayEventData "(TagName=\"\")" 0 0 0 0 "()" "(GameplayTags=)" "(GameplayTags=)" (Math|Conversions|ToFloat(Integer) _returnvalue_4)))
      (if (== _returnvalue_4 0)
        (if (>= _returnvalue_3 0)
          (Utilities|Time|SetTimerbyEvent (delegate Custom|FallbackRemoveWeapon) _returnvalue_3))))))

(event Custom|FallbackRemoveWeapon
  (Inventory|RemoveItemFromPlayerByGuid (Game|GetInstigatorController) (FortWeapon|GetInventoryGUID) -1 true))

(event FortWeapon|EventOnLocalUnEquip
  ;[1 1 1] Unequipped - Safety (if unequipped with no ammo...somehow, remove item)
  (CallFunction|RemoveItemOnNoAmmoRemaining))

(event Custom|Remove%20Item%20on%20No%20Ammo%20Remaining
  ;[1 1 1] Persist in Inventory when final stack empty in WID does not handle removing a custom ammo weapon with 0 ammo
  (if (<= (FortWeapon|GetRemainingAmmo) 0)
    (bind _asfort_player_pawn (Utilities|Casting|CastToFortPlayerPawn (Game|GetInstigator)))
    (Inventory|RemoveItemFromPlayerByGuid (Fortnite|GetFortPlayerControllerfromActor _asfort_player_pawn) (FortWeapon|GetInventoryGUID) 1 true)))

(event EventPlayWeaponFireCustomAudio
  (params (bPersistentFire :type Boolean) (bSecondaryFire :type Boolean))
  (bind _playerpawn (Variables|CachedVariables|GetPlayerPawn))
  (bind _returnvalue (Pawn|IsLocallyViewed _playerpawn))
  (bind _chargestage (Variables|Default|GetChargeStage))
  (bind _high_charge_weapon_fire_1p (Variables|Sounds|GetHigh%20Charge%20Weapon%20Fire%201p))
  (bind _low_charge_weapon_fire_1p (Variables|Sounds|GetLow%20Charge%20Weapon%20Fire%201p))
  (bind _returnvalue_2 (Utilities|Select :Index _chargestage :Option%200 _low_charge_weapon_fire_1p :Option%201 _low_charge_weapon_fire_1p :Option%202 _high_charge_weapon_fire_1p :Option%203 _high_charge_weapon_fire_1p))
  (bind _high_charge_weapon_fire_3p (Variables|Sounds|GetHigh%20Charge%20Weapon%20Fire%203p))
  (bind _low_charge_weapon_fire_3p (Variables|Sounds|GetLow%20Charge%20Weapon%20Fire%203p))
  (bind _returnvalue_3 (Utilities|Select :Index _chargestage :Option%200 _low_charge_weapon_fire_3p :Option%201 _low_charge_weapon_fire_3p :Option%202 _high_charge_weapon_fire_3p :Option%203 _high_charge_weapon_fire_3p))
  (bind _returnvalue_4 (select _returnvalue _returnvalue_2 _returnvalue_3))
  ; Select sound based on perspective and charge state. Retrigger sound if selected sound is already active and is the same sound, else spawn a new sound.
  (|CallParentFunction bPersistentFire bSecondaryFire)
  (bind _weapon_fire_audio (Variables|Default|GetWeapon%20Fire%20Audio)
    (:then
      (if (not (and (== (Variables|/Script/Engine.AudioComponent|GetSound _weapon_fire_audio) _returnvalue_4) (not (Variables|Sounds|GetbIsAudioComponentExpiring))))
        (bind _returnvalue_5 (Audio|SpawnSoundAttached _returnvalue_4 (Variables|Visual|GetWeaponMesh)))
        (bind _output_get (Variables|Default|SetWeapon%20Fire%20Audio _returnvalue_5))
        (MetaSoundOutput|WatchOutput (WorldSubsystems|GetMetaSoundOutputSubsystem) _output_get "bIsAlmostFinished" (delegate Custom|CustomEvent) "None" "None" :!class /Script/MetasoundEngine.MetaSoundOutputSubsystem)
        (Variables|Sounds|SetbIsAudioComponentExpiring false))
      (Audio|Parameter|ExecuteTriggerParameter (Variables|Default|GetWeapon%20Fire%20Audio) "Retrigger")
      ; Shot AV
      (SoundIndicator|ShowSoundIndicator (Variables|SoundIndicator|GetSoundIndicatorComponent) (Transformation|GetActorLocation (Variables|CachedVariables|GetPlayerPawn)) (Variables|Sounds|GetHigh%20Charge%20Weapon%20Fire%203p) 1.0 0))
    (:else
      _)))

(event Custom|CustomEvent
  (params (OutputName :type Name) (Output :type Meta%20Sound%20Output%20Structure%20%28by%20ref%29))
  ; Select sound based on perspective and charge state. Retrigger sound if selected sound is already active and is the same sound, else spawn a new sound.
  (bind (_success _returnvalue) (MetaSoundOutput|GetBool Output))
  (Variables|Sounds|SetbIsAudioComponentExpiring _returnvalue))

(event EventEndPlay
  (params (EndPlayReason :type EEndPlayReason%20Enum :default "Destroyed"))
  (|CallParentFunction EndPlayReason)
  (bind _weapon_fire_audio (Variables|Default|GetWeapon%20Fire%20Audio)
    (:then
      (Audio|Components|Audio|FadeOut _weapon_fire_audio 0.2 0.0 "Linear" :!class /Script/Engine.AudioComponent))
    (:else)))

(fn CacheVariables
  (bind _asfort_player_pawn (Utilities|Casting|CastToFortPlayerPawn (Game|GetInstigator)))
  (Variables|CachedVariables|SetPlayerPawn _asfort_player_pawn))

(fn GetAimRotOverride
  (params (Weapon :type Fort%20Weapon%20Object%20Reference) (TargetingSource :type EFortAbilityTargetingSource%20Enum :default "Camera"))
  (bind (_returnvalue_location _returnvalue_rotation) (Math|Transform|BreakTransform (Targeting|GetTargetingTransform (Variables|CachedVariables|GetPlayerPawn) TargetingSource 0 :!class /Script/FortniteGame.FortPawn)))
  (if (>= (FortWeapon|GetChargePercent) 1.0)
    (LocalVariables|SetPitch (ScalableFloat|GetCachedScalableFloatValue (Variables|Aiming|GetFullChargeMuzzlePitchAdjustment)))
    (LocalVariables|SetYaw (ScalableFloat|GetCachedScalableFloatValue (Variables|Aiming|GetFullChargeMuzzleYawPitchAdjustment)))
    (else
      (LocalVariables|SetPitch (ScalableFloat|GetCachedScalableFloatValue (Variables|Aiming|GetMuzzlePitchAdjustment)))
      (LocalVariables|SetYaw (ScalableFloat|GetCachedScalableFloatValue (Variables|Aiming|GetMuzzleYawAdjustment)))))
  (return (Math|Rotator|CombineRotators (Math|Rotator|CombineRotators _returnvalue_rotation (Math|Vector|RotatorfromAxisandAngle (Math|Vector|GetRightVector _returnvalue_rotation) (* (LocalVariables|GetPitch) -1.0))) (Math|Vector|RotatorfromAxisandAngle (Math|Vector|GetUpVector _returnvalue_rotation) (LocalVariables|GetYaw))) true))

(fn GetAnimData
  (return (Variables|Default|GetChargeStage) (Variables|Default|GetOnChargeEnd)))

(fn GetChargeStage
  (bind _returnvalue (FortWeapon|GetChargePercent))
  (if (== _returnvalue 0)
    (Variables|Default|SetChargeStage 0)
    (elif (and (> _returnvalue 0) (< _returnvalue 0.5))
      (Variables|Default|SetChargeStage 1)
      (elif (and (>= _returnvalue 0.5) (< _returnvalue 1.0))
        (Variables|Default|SetChargeStage 2)
        (elif (>= _returnvalue 1.0)
          (Variables|Default|SetChargeStage 3))))))

(fn HasScopeHUDReticleTag
  (params (Tag :type Gameplay%20Tag%20Structure%20%28by%20ref%29))
  (return false))

(fn ConstructionScript
  (|CallParentFunction)
  (Variables|Default|SetRearMuzzleFXSystem "/NanaimoBar/Weapons/FORT_Melee/Pickaxe_AshtonBoardwalk/FX/p_EMPTY.p_EMPTY")
  (Variables|Effects|Muzzle|SetMuzzleWindParticleSystem "/NanaimoBar/Weapons/FORT_Melee/Pickaxe_AshtonBoardwalk/FX/p_EMPTY.p_EMPTY"))
