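; Chargeable Fortnite projectile for the NanaimoBar weapon that explodes on touch or when its lifespan ends, and notifies the instigator pawn on spawn.
; Floored charge percent switches capsule and trail scale, while charge also drives impact FX size and a fully charged async-explosion preset.
; Water hits skip detonation above a Z-velocity threshold, and a delayed timer enables water blocking, exploding the projectile if already submerged.
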(event EventReplacedPredictedProjectile
  (params (PredictedProjectile :type Fort%20Projectile%20Base%20Object%20Reference))
  (Components|Activation|Deactivate (Variables|Default|GetProj_TrailFX)))

(event EventBeginPlay
  (CallFunction|AdjustProjectileFxScale)
  ; VFX Treatment for Water hit
  (Utilities|Time|SetTimerbyEvent (EventDispatchers|CreateEvent :!event Custom|WaterExplodeDelay) (Variables|Default|GetWaterInteractionDelay))
  ; Event to account for shot and eventually discard weapon when out of ammo
  (bind _asfort_player_pawn (Utilities|Casting|CastToFortPlayerPawn (Game|GetInstigator)))
  (Ability|SendGameplayEventtoActor _asfort_player_pawn "(TagName=\"NanaimoBar.Item.Projectile.Spawned\")")
  ; Auto-destruct after some amount of time
  (Utilities|FlowControl|Delay (ScalableFloat|GetCachedScalableFloatValue (Variables|Default|GetProjectileLifespan)))
  (CallFunction|ImpactFX (Projectile|GetCurrentLocation))
  (CallFunction|Explode))

(event EventOnTouched
  (params (OtherActor :type Actor%20Object%20Reference) (OtherComp :type Primitive%20Component%20Object%20Reference) (HitResult :type Hit%20Result%20Structure%20%28by%20ref%29) (bIsOverlap :type Boolean))
  (bind (_returnvalue_x _returnvalue_y _returnvalue_z) (Math|Vector|BreakVector (Transformation|GetVelocity)))
  (bind (_bblockinghit _binitialoverlap _time _distance _location _impactpoint _normal _impactnormal _physmat _hitactor) (Collision|BreakHitResult HitResult))
  ; On Touched - No Stick Check
  (|CallParentFunction OtherActor OtherComp HitResult bIsOverlap)
  (bind _aswater_body (Utilities|Casting|CastToWaterBody OtherActor)
    (:then
      ; Water Surface
      ;; This is to allow the player to shoot another player while both are in water, by avoiding the explosion when the velocity is above a low threshold
      (if (>= _returnvalue_z (ScalableFloat|GetCachedScalableFloatValue (Variables|Default|GetWaterParallelShot)))
        (Variables|Projectile|SetbStopSimulatingOnHit false)
        (else
          (Variables|Default|SetLocalStopSimulatingOnHit true)
          (CallFunction|ImpactFXWater (WaterBody|GetWaterSurfaceInfoatLocation (Water|GetWaterBodyComponent _aswater_body) (Transformation|GetWorldLocation (Variables|Collision|GetCapsuleComponent)) false) (.OutWaterSurfaceNormal (WaterBody|GetWaterSurfaceInfoatLocation (Water|GetWaterBodyComponent _aswater_body) (Transformation|GetWorldLocation (Variables|Collision|GetCapsuleComponent)) false)))
          (CallFunction|Explode))))
    (:CastFailed
      (Variables|Default|SetLocalStopSimulatingOnHit true)
      (CallFunction|ImpactFX _impactpoint _normal _hitactor)
      (CallFunction|Explode))))

(event Custom|WaterExplodeDelay
  ; Comment
  (Collision|SetCollisionResponsetoChannel (Variables|Collision|GetCapsuleComponent) "ECC_GameTraceChannel18" "ECR_Block")
  (Variables|Projectile|SetTouchWaterBehavior "StopIfStopSimulatingOnHit")
  (if (Water|IsinWater (Variables|Water|GetWaterInteractionComponent))
    (if (not (Delivery|IsBeingKilled))
      (CallFunction|Explode))))

(fn AdjustProjectileFxScale
  (bind _chargepercent (Variables|FortProjectileBase|GetChargePercent))
  (bind _returnvalue (Math|Float|Floor _chargepercent))
  (bind _returnvalue_2 (Math|Conversions|ToFloat(Integer) _returnvalue))
  (bind _proj_trailfx (Variables|Default|GetProj_TrailFX))
  (Transformation|SetRelativeScale3D (Variables|Collision|GetCapsuleComponent) (Math|Conversions|ToVector(Float) (Math|Float|Lerp (ScalableFloat|GetCachedScalableFloatValue (Variables|Scale|GetScaleOnMinCharge)) (ScalableFloat|GetCachedScalableFloatValue (Variables|Scale|GetScaleOnMaxCharge)) _returnvalue_2)))
  (Niagara|SetNiagaraVariable(Float) _proj_trailfx "FXScale" (Math|Float|Lerp (ScalableFloat|GetCachedScalableFloatValue (Variables|Scale|GetFxScaleOnMinCharge)) (ScalableFloat|GetCachedScalableFloatValue (Variables|Scale|GetFxScaleOnMaxCharge)) _returnvalue_2))
  (Niagara|SetNiagaraVariable(Float) _proj_trailfx "FXLerp" _returnvalue_2))

(fn AdjustProjectileGravity
  (if (>= (Variables|FortProjectileBase|GetChargePercent) 1.0)
    (Projectile|SetGravityScale 0.0)))

(fn Explode
  (if (> (ScalableFloat|GetCachedScalableFloatValue (Variables|Default|GetbUseAsyncTrace)) 0)
    (if (>= (Variables|FortProjectileBase|GetChargePercent) 1.0)
      (Variables|FortGameplayEffectDeliveryActor|SetAsyncExplosionTargetingPreset (Variables|Default|GetFullyChargedAsyncExplosionTargetingPreset)))
    (Delivery|DoExplosionAsync)
    (else
      (Delivery|DoExplosion))))

(fn ImpactFX
  (params (Location :type Vector) (Normal :type Vector) (HitActor :type Actor%20Object%20Reference))
  (bind (_asfort_vehicle_interface _bsuccess) (Utilities|Casting|CastToFortVehicleInterface HitActor))
  (bind _chargepercent (Variables|FortProjectileBase|GetChargePercent))
  (if (Variables|Rendering|GetbHidden)
    _
    (elif (Math|Boolean|ANDBoolean (not (Fortnite|IsMobilePlatform)) (not (Class|PlatformHelpersLibrary|IsPcPerfMode)) (> (Rendering|GetRenderingDetailMode) 1))
      _))
  (bind _returnvalue (Effects|Components|ParticleSystem|SpawnFXAtLocation "/NanaimoBar/Assets/Fx/NS_NanaimoBar_Projectile_Hit.NS_NanaimoBar_Projectile_Hit" Location (Math|Conversions|RotationFromXVector Normal)))
  (bind _returnvalue_2 (GameplayTags|HasMatchingGameplayTag(Message) HitActor "(TagName=\"Athena.Gameplay.Train\")"))
  (Utilities|IsValid _returnvalue
    (:Is%20Valid
      (Effects|Components|ParticleSystem|SetBoolParameter _returnvalue "IsTrain" _returnvalue_2))
    (:Is%20Not%20Valid
      (Effects|Components|ParticleSystem|SetBoolParameter _returnvalue "IsTrain" _bsuccess)))
  (Effects|Components|ParticleSystem|SetFloatParameter _returnvalue "FXScale" (Math|Conversions|ToFloat(Integer) (Math|Float|Floor _chargepercent)))
  (Effects|Components|ParticleSystem|SetFloatParameter _returnvalue "FXRadius" (ScalableFloat|GetValueAtLevel(double-precision) (Variables|Default|GetSize) (* (Math|Conversions|ToFloat(Integer) (Math|Float|Floor (* _chargepercent 2.0))) 0.5)))
  (GameplayCue|ExecuteGameplayCue%28Burst%29Local self (Variables|Default|GetGCN_ScreenFX) (Ability|GameplayCue|MakeGameplayCueParameters 0.0)) ; Cam shake + controller vibration
)

(fn ImpactFX_Water
  (params (Location :type Vector) (Normal :type Vector))
  (bind _chargepercent (Variables|FortProjectileBase|GetChargePercent))
  (if (Variables|Rendering|GetbHidden)
    _
    (elif (Math|Boolean|ANDBoolean (not (Fortnite|IsMobilePlatform)) (not (Class|PlatformHelpersLibrary|IsPcPerfMode)) (> (Rendering|GetRenderingDetailMode) 1))
      _))
  (bind _returnvalue (Effects|Components|ParticleSystem|SpawnFXAtLocation "/NanaimoBar/Assets/Fx/NS_NanaimoBar_Projectile_Water_Hit.NS_NanaimoBar_Projectile_Water_Hit" Location (Math|Conversions|RotationFromXVector Normal)))
  (Effects|Components|ParticleSystem|SetFloatParameter _returnvalue "FXScale" (Math|Conversions|ToFloat(Integer) (Math|Float|Floor _chargepercent)))
  (Effects|Components|ParticleSystem|SetFloatParameter _returnvalue "FXRadius" (ScalableFloat|GetValueAtLevel(double-precision) (Variables|Default|GetSize) (* (Math|Conversions|ToFloat(Integer) (Math|Float|Floor (* _chargepercent 2.0))) 0.5))))

(fn ConstructionScript
  (|CallParentFunction))
