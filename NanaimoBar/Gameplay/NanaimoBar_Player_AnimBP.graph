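# Player animation Blueprint for the NanaimoBar chargeable weapon, filling Item Override layers on the Fortnite player skeleton.
# The final-pose layer runs a charge state machine, keeps the incoming lower body through movement-dependent blend profiles, then blends in an aim offset.
# The full-body layer adds variable-driven pelvis rotation.
# FireWeapon_SetupFireAnim picks an additive fire animation by ChargeStage with inertial blending; other layers have no logic.
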AnimBlueprint /NanaimoBar/Gameplay/NanaimoBar_Player_AnimBP
target_skeleton=&Fortnite_M_Avg_Player_Skeleton
graphs [
 AnimGraph {nodes [Output Pose()]}
 EventGraph {links [], nodes []}
 FireWeapon_SetupFireAnim {nodes [
   n1 = FireWeapon_SetupFireAnim!() -> (then, Context:"Anim Update Context Structure (by ref)", Node:"Anim Node Reference Structure (by ref)") {is=title K2Node_FunctionEntry}
   n2 = Convert to Sequence Player!(Node:"Anim Node Reference Structure (by ref)"=@n1.Node) -> (SequencePlayer:Sequence Player Reference Structure, Result:bool)
   > Branch(Condition=@n2.Result)
   > SetSequenceWithInertialBlending!(
    execute
    UpdateContext:"Anim Update Context Structure (by ref)"=@n1
    SequencePlayer:"Sequence Player Reference Structure (by ref)"=@n2
    Sequence:Anim Sequence Base Object Reference=Select!(
     Option 0:Anim Sequence Base Object Reference=/NanaimoBar/Assets/Animation/Animation/NanaimoBar_Fire_NonTargeted_CMM_Add
     Option 1:Anim Sequence Base Object Reference=&NanaimoBar_Fire_NonTargeted_CMM_Add
     Option 2:Anim Sequence Base Object Reference=/NanaimoBar/Assets/Animation/Animation/NanaimoBar_Fire_Stage02_CMM_Add
     Option 3:Anim Sequence Base Object Reference=/NanaimoBar/Assets/Animation/Animation/NanaimoBar_Fire_Stage03_CMM_Add
     Index:int=Get ChargeStage() -> (ChargeStage:int=0)
    ) -> (ReturnValue:Anim Sequence Base Object Reference)
    BlendTime:float=0.2
   ) -> (then, ReturnValue:Sequence Player Reference Structure)
   > Return Node()
  ]}
 ItemAimOffset {nodes [
   Input Pose!() -> (Pose, AimOffsetAlpha:double, Yaw:double, Pitch:double) {name=InPose_AimOffset}
   Input Pose!() -> (Pose) {name=InPose_PreAimOffsets}
   Output Pose() {layerGroup=Item Override}
  ]}
 ItemFallAdditive {nodes [
   Input Pose() {name=InPoseItemFallAdditive}
   Output Pose() {layerGroup=Item Override}
  ]}
 ItemFallLandAdditive {nodes [
   Input Pose() {name=InPoseItemFallLandAdditive}
   Output Pose() {layerGroup=Item Override}
  ]}
 ItemFallLevel2Override {nodes [
   Input Pose() {name=InPose_FallLevel2}
   Output Pose() {layerGroup=Item Override}
  ]}
 ItemFinalPoseOverride {nodes [
   n1 = Use cached pose 'PreAimOffsets'()
   n2 = Get AimPitch()
   n3 = Get AimYaw() -> (AimYaw:float)
   Output Pose(Result=Two Way Blend(
     A=Use cached pose 'PreAimOffsets'()
     B=AimOffset Player 'NanaimoBar_AO'!(X:float=@n3, Y:float=@n2, BasePose:Pose Link Structure=@n1, Alpha:float=1) -> (Pose:Pose Link Structure) {blendSpace=/NanaimoBar/Assets/Animation/AO/NanaimoBar_AO, alphaBoolBlend {blendOption=Linear}}
     Alpha=Get AimOffsetWeight() -> (AimOffsetWeight:float)
    ) {blendNode {alphaBoolBlend {blendOption=Linear}}}) {layerGroup=Item Override}
   Save cached pose 'FB Input'(Pose=Input Pose() {name=InFinalPose}) {cacheName=FB Input}
   AnimGraphNode_StateMachine_0 = ChargeSM() {is=title AnimGraphNode_StateMachine}
   Save cached pose 'Charge SM'(Pose=@AnimGraphNode_StateMachine_0) {cacheName=Charge SM}
   n4 = Profile Blend!(BasePose:Pose Link Structure=Use cached pose 'Charge SM'(), BlendPose:Pose Link Structure=Use cached pose 'FB Input'(), BlendWeight:float=1) -> (Pose:Pose Link Structure) {
    is=title AnimGraphNode_BlendProfileLayeredBlend
    blendProfileAsset=/Game/Characters/Player/BlendMasks/LowerBodyPartialUpperBody_BlendMask
    rotationBlendSpace=MeshSpace
    blendPose {linkId=16842751}
   }
   n5 = Profile Blend(BasePose=Use cached pose 'Charge SM'(), BlendPose=Use cached pose 'FB Input'(), BlendWeight=1) {blendProfileAsset=/Game/Characters/Player/BlendMasks/LowerBody_BlendMask, rotationBlendSpace=MeshSpace, blendPose {linkId=16842751}}
   Save cached pose 'PreAimOffsets'(Pose=Blend Poses by bool(BlendPose_0=@n4, BlendPose_1=@n5, bActiveValue=Get bIsMoving2D()) {blendType=Linear}) {cacheName=PreAimOffsets}
  ]}
 ChargeSM {
  entry=PassThrough
  path=ItemFinalPoseOverride.AnimGraphNode_StateMachine_0.ChargeSM
  ids=id
  nodes [
   Exit Alias() {is=title AnimStateAliasNode, also {aliases [Charge 1, Charge 3 Intro, Charge 3 Loop]}}
   AnimStateNode_0 = PassThrough() {is=title AnimStateNode}
   AnimStateNode_1 = Charge 1() {is=title AnimStateNode}
   AnimStateNode_2 = Charge 3 Intro() {is=title AnimStateNode}
   AnimStateNode_3 = Charge 3 Loop() {is=title AnimStateNode}
   AnimStateTransitionNode_0 = AnimStateTransitionNode()
   AnimStateTransitionNode_1 = AnimStateTransitionNode()
   AnimStateTransitionNode_2 = AnimStateTransitionNode() {bAutomaticRuleBasedOnSequencePlayerInState=true}
   AnimStateTransitionNode_3 = AnimStateTransitionNode() {crossfadeDuration=0}
  ]
  links [
   {
    dst=Charge 3 Intro
    rule=ItemFinalPoseOverride.AnimGraphNode_StateMachine_0.ChargeSM.AnimStateTransitionNode_1.Transition
    src=Charge 1
    via=AnimStateTransitionNode_1
   }
   {
    dst=Charge 3 Loop
    rule=ItemFinalPoseOverride.AnimGraphNode_StateMachine_0.ChargeSM.AnimStateTransitionNode_2.Transition
    src=Charge 3 Intro
    via=AnimStateTransitionNode_2
   }
   {
    dst=PassThrough
    rule=ItemFinalPoseOverride.AnimGraphNode_StateMachine_0.ChargeSM.AnimStateTransitionNode_3.Transition
    src=Exit Alias
    via=AnimStateTransitionNode_3
   }
   {
    dst=Charge 1
    rule=ItemFinalPoseOverride.AnimGraphNode_StateMachine_0.ChargeSM.AnimStateTransitionNode_0.Transition
    src=PassThrough
    via=AnimStateTransitionNode_0
   }
  ]
 }
 PassThrough {
  path=ItemFinalPoseOverride.AnimGraphNode_StateMachine_0.ChargeSM.AnimStateNode_0.PassThrough
  nodes [Output Animation Pose(Result=Use cached pose 'FB Input'())]
 }
 Charge 1 {
  path=ItemFinalPoseOverride.AnimGraphNode_StateMachine_0.ChargeSM.AnimStateNode_1.Charge 1
  nodes [Output Animation Pose(Result=Sequence Player 'NanaimoBar_Charge_Stage01_CMM'() {bLoopAnimation=false, sequence=/NanaimoBar/Assets/Animation/Animation/NanaimoBar_Charge_Stage01_CMM})]
 }
 Charge 3 Intro {
  path=ItemFinalPoseOverride.AnimGraphNode_StateMachine_0.ChargeSM.AnimStateNode_2.Charge 3 Intro
  nodes [Output Animation Pose(Result=Sequence Player 'NanaimoBar_Charge_Shake_INTRO_03_CMM'() {bLoopAnimation=false, sequence=/NanaimoBar/Assets/Animation/Animation/NanaimoBar_Charge_Shake_INTRO_03_CMM})]
 }
 Charge 3 Loop {
  path=ItemFinalPoseOverride.AnimGraphNode_StateMachine_0.ChargeSM.AnimStateNode_3.Charge 3 Loop
  nodes [Output Animation Pose(Result=Sequence Player 'NanaimoBar_Charge_Stage03_Loop_CMM'() {sequence=/NanaimoBar/Assets/Animation/Animation/NanaimoBar_Charge_Stage03_Loop_CMM})]
 }
 Transition {
  path=ItemFinalPoseOverride.AnimGraphNode_StateMachine_0.ChargeSM.AnimStateTransitionNode_0.Transition
  nodes [Result(bCanEnterTransition=Get bChargeSM_PassthroughToCharge1() -> (bChargeSM_PassthroughToCharge1:bool))]
 }
 Transition {
  path=ItemFinalPoseOverride.AnimGraphNode_StateMachine_0.ChargeSM.AnimStateTransitionNode_1.Transition
  nodes [Result(bCanEnterTransition=Get bChargeSM_Charge1ToCharge3Intro() -> (bChargeSM_Charge1ToCharge3Intro:bool))]
 }
 Transition {path=ItemFinalPoseOverride.AnimGraphNode_StateMachine_0.ChargeSM.AnimStateTransitionNode_2.Transition, nodes [Result()]}
 Transition {
  path=ItemFinalPoseOverride.AnimGraphNode_StateMachine_0.ChargeSM.AnimStateTransitionNode_3.Transition
  nodes [Result(bCanEnterTransition=Get bChargeSM_AliasToPassthrough() -> (bChargeSM_AliasToPassthrough:bool))]
 }
 ItemFullBodyAimOffsets {nodes [
   Input Pose!() -> (Pose, FullBody_AimOffsetAlpha:double, Aim_Yaw:double, Aim_Pitch:double) {name=InPoseFullBodyAimOffsets}
   Input Pose!() -> (Pose) {name=InPosePreFullBodyAimOffsets}
   Output Pose() {layerGroup=Item Override}
  ]}
 ItemFullBodyOverride {nodes [
   n1 = Local To Component(LocalPose=Input Pose() {name=InPoseFullBody})
   n2 = Get CustomMeleeTwist() -> (CustomMeleeTwist:float)
   n3 = Get CustomMeleePitch() -> (CustomMeleePitch:float)
   n4 = Get CustomMeleeRoll() -> (CustomMeleeRoll:float)
   n5 = Get CustomMeleeTwist_Alpha() -> (CustomMeleeTwist_Alpha:float)
   Output Pose(Result=Component To Local(ComponentPose="Transform (Modify) Bone - Bone: pelvis"!(ComponentPose=@n1, Alpha=@n5, Rotation_Roll:float=@n4, Rotation_Pitch:float=@n3, Rotation_Yaw:float=@n2) -> (Pose) {rotationMode=BMM_Additive, alphaBoolBlend {blendOption=Linear}, boneToModify {boneName=pelvis}})) {layerGroup=Item Override}
  ]}
 ItemIdleOverride {nodes [
   Input Pose() {name=InPose_Idle}
   Output Pose() {layerGroup=Item Override}
  ]}
 ItemJetPackJumpAdditive {nodes [Output Pose() {layerGroup=Item Override}]}
 ItemJetPackStartAdditive {nodes [Output Pose() {layerGroup=Item Override}]}
 ItemJumpLoopAdditive {nodes [
   Input Pose() {name=InPoseItemJumpLoopAdditive}
   Output Pose() {layerGroup=Item Override}
  ]}
 ItemJumpOverride {nodes [
   Input Pose() {name=InPoseJump}
   Output Pose() {layerGroup=Item Override}
  ]}
 ItemJumpUpAdditive {nodes [
   Input Pose() {name=InPoseItemJumpUpAdditive}
   Output Pose() {layerGroup=Item Override}
  ]}
 ItemPreAimOffsets {nodes [
   Input Pose() {name=InPosePreAimOffsets}
   Output Pose() {layerGroup=Item Override}
  ]}
 ItemPreIK {nodes [
   Input Pose() {name=InPose_PreIK}
   Output Pose() {layerGroup=Item Override}
  ]}
 ItemRightHandIK {nodes [
   Input Pose() {name=InPosePreRightHandIK}
   Input Pose() {name=InPosePostRightHandIK}
   Output Pose() {layerGroup=Item Override}
  ]}
 ItemSkeletalControl {nodes [
   Input Pose()
   Output Pose() {layerGroup=Item Override}
  ]}
 ItemSwimJumpFallLoop {nodes [Output Pose() {layerGroup=Item Override}]}
 ItemSwimJumpSurfaceEnd {nodes [Output Pose() {layerGroup=Item Override}]}
 ItemSwimJumpSurfaceLoop {nodes [Output Pose() {layerGroup=Item Override}]}
 ItemUpperBody {nodes [
   Input Pose!() -> (Pose, InputParam:Fort Anim Input Adjusted Aim Structure) {name=InPoseUpperBody}
   Output Pose() {layerGroup=Item Override}
  ]}
 ItemZipLineStartAdditive {nodes [Output Pose() {layerGroup=Item Override}]}
]
variables []
