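# Animation Blueprint for the NanaimoBar gauntlet weapon, animating idle, staged charging and firing.
# AdditiveShake_SM's idle/charge pose passes through DefaultSlot montages, bypassed by bHideWeaponEquipAnim, into the FiringOverride state machine's Passthrough state.
# Charge end triggers Fire, which alternates with FireAgain on interrupts; SetupFireAnim selects each sequence by ChargeStage as the player becomes relevant.
# An additive bone-space translation offsets the root before output.
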AnimBlueprint /NanaimoBar/Gameplay/NanaimoBar_Weapon_AnimBP
target_skeleton=/NanaimoBar/Assets/Meshes/NanaimoBar_Gauntlet_Skeleton
graphs [
 AnimGraph {nodes [
   AnimGraphNode_StateMachine_2 = FiringOverride() {is=title AnimGraphNode_StateMachine}
   n1 = "Transform (Modify) Bone - Bone: root"(ComponentPose=Local To Component(LocalPose=@AnimGraphNode_StateMachine_2), Translation="-0.9,-0.9,-3", Rotation="0,0,0") {
    translationMode=BMM_Additive
    translationSpace=BCS_BoneSpace
    alphaBoolBlend {blendOption=Linear}
    boneToModify {boneName=root}
    translation=<-0.9 -0.9 -3>
   }
   Output Pose(Result=Component To Local(ComponentPose=@n1))
   n2 = Use cached pose 'PostAdditiveShake'()
   Save cached pose 'PostMontage'(Pose=Blend Poses by bool(
     BlendPose_0=Use cached pose 'PostAdditiveShake'()
     BlendPose_1=Slot 'DefaultSlot'(Source=@n2)
     bActiveValue=Get bHideWeaponEquipAnim() -> (bHideWeaponEquipAnim:bool)
    ) {blendType=Linear}) {cacheName=PostMontage}
   AnimGraphNode_StateMachine_0 = AdditiveShake_SM() {is=title AnimGraphNode_StateMachine, bAllowConduitEntryStates=true}
   Save cached pose 'PostAdditiveShake'(Pose=@AnimGraphNode_StateMachine_0) {cacheName=PostAdditiveShake}
  ]}
 AdditiveShake_SM {
  entry=Idle
  path=AnimGraph.AnimGraphNode_StateMachine_0.AdditiveShake_SM
  ids=id
  nodes [
   Exit Charge Alias() {is=title AnimStateAliasNode, also {aliases [Charge Intro, Charge_Loop]}}
   AnimStateNode_0 = Idle()
   AnimStateNode_1 = Charge_Loop() {is=title AnimStateNode, bAlwaysResetOnEntry=true}
   AnimStateNode_2 = Charge Intro() {is=title AnimStateNode, bAlwaysResetOnEntry=true}
   AnimStateTransitionNode_0 = AnimStateTransitionNode()
   AnimStateTransitionNode_1 = AnimStateTransitionNode() {bAutomaticRuleBasedOnSequencePlayerInState=true}
   AnimStateTransitionNode_5 = AnimStateTransitionNode()
   AnimStateTransitionNode_8 = AnimStateTransitionNode()
  ]
  links [
   {
    dst=Charge_Loop
    rule=AnimGraph.AnimGraphNode_StateMachine_0.AdditiveShake_SM.AnimStateTransitionNode_1.Transition
    src=Charge Intro
    via=AnimStateTransitionNode_1
   }
   {
    dst=Charge Intro
    rule=AnimGraph.AnimGraphNode_StateMachine_0.AdditiveShake_SM.AnimStateTransitionNode_8.Transition
    src=Charge_Loop
    via=AnimStateTransitionNode_8
   }
   {
    dst=Idle
    rule=AnimGraph.AnimGraphNode_StateMachine_0.AdditiveShake_SM.AnimStateTransitionNode_0.Transition
    src=Exit Charge Alias
    via=AnimStateTransitionNode_0
   }
   {
    dst=Charge Intro
    rule=AnimGraph.AnimGraphNode_StateMachine_0.AdditiveShake_SM.AnimStateTransitionNode_5.Transition
    src=Idle
    via=AnimStateTransitionNode_5
   }
  ]
 }
 Idle {
  path=AnimGraph.AnimGraphNode_StateMachine_0.AdditiveShake_SM.AnimStateNode_0.Idle
  nodes [Output Animation Pose(Result=Sequence Player 'NanaimoBar_Idle_Loop_WEP'() {sequence=/NanaimoBar/Assets/Animation/Animation/NanaimoBar_Idle_Loop_WEP})]
 }
 Charge_Loop {
  path=AnimGraph.AnimGraphNode_StateMachine_0.AdditiveShake_SM.AnimStateNode_1.Charge_Loop
  nodes [Output Animation Pose(Result=Sequence Player 'NanaimoBar_Charge_Stage03_Loop_WEP'() {sequence=/NanaimoBar/Assets/Animation/Animation/NanaimoBar_Charge_Stage03_Loop_WEP})]
 }
 Charge Intro {
  path=AnimGraph.AnimGraphNode_StateMachine_0.AdditiveShake_SM.AnimStateNode_2.Charge Intro
  nodes [
   n1 = Sequence Player 'NanaimoBar_Charge_Stage01_WEP'() {
    bLoopAnimation=false
    groupName=Charge_Stage_1
    method=SyncGroup
    sequence=/NanaimoBar/Assets/Animation/Animation/NanaimoBar_Charge_Stage01_WEP
   }
   n2 = Sequence Player 'NanaimoBar_Charge_Stage01_WEP'() {
    bLoopAnimation=false
    groupName=Charge_Stage_1
    method=SyncGroup
    sequence=&NanaimoBar_Charge_Stage01_WEP
   }
   n3 = Sequence Player 'NanaimoBar_Charge_Stage01_WEP'() {
    bLoopAnimation=false
    groupName=Charge_Stage_1
    method=SyncGroup
    sequence=&NanaimoBar_Charge_Stage01_WEP
   }
   n4 = Sequence Player 'NanaimoBar_Charge_Stage03_WEP'() {
    bLoopAnimation=false
    groupName=Charge_Stage_1
    method=SyncGroup
    sequence=/NanaimoBar/Assets/Animation/Animation/NanaimoBar_Charge_Stage03_WEP
   }
   n5 = Blend Poses by int!(
    BlendPose_0:Pose Link Structure=@n2
    BlendPose_1:Pose Link Structure=@n1
    BlendPose_2:Pose Link Structure=@n3
    BlendPose_3:Pose Link Structure=@n4
    BlendTime_0:float=0.1
    BlendTime_1:float=0.1
    BlendTime_2:float=0.1
    BlendTime_3:float=0.1
    ActiveChildIndex:int=Get ChargeStage() -> (ChargeStage:int=0)
   ) -> (Pose:Pose Link Structure) {
    is=title AnimGraphNode_BlendListByInt
    blendType=Linear
    blendPose [{linkId=-1, sourceLinkId=-1}, {linkId=-1, sourceLinkId=-1}, {linkId=-1, sourceLinkId=-1}, {linkId=-1, sourceLinkId=-1}]
    blendTime [0.1,0.1,0.1,0.1]
   }
   Output Animation Pose(Result=@n5) {node {stateEntryFunction {functionName=SetInitialChargeState}}, stateEntryFunction {bSelfContext=true, memberName=SetInitialChargeState}}
  ]
 }
 Transition {
  path=AnimGraph.AnimGraphNode_StateMachine_0.AdditiveShake_SM.AnimStateTransitionNode_0.Transition
  nodes [Result(bCanEnterTransition=Get bChargeSM_ChargeAliasToIdle() -> (bChargeSM_ChargeAliasToIdle:bool))]
 }
 Transition {path=AnimGraph.AnimGraphNode_StateMachine_0.AdditiveShake_SM.AnimStateTransitionNode_1.Transition, nodes [Result()]}
 Transition {
  path=AnimGraph.AnimGraphNode_StateMachine_0.AdditiveShake_SM.AnimStateTransitionNode_5.Transition
  nodes [Result(bCanEnterTransition=Get bChargeSM_IdleToCharge() -> (bChargeSM_IdleToCharge:bool))]
 }
 Transition {
  path=AnimGraph.AnimGraphNode_StateMachine_0.AdditiveShake_SM.AnimStateTransitionNode_8.Transition
  nodes [Result(bCanEnterTransition=Get bChargeSM_ChargeLoopToChargeIntro() -> (bChargeSM_ChargeLoopToChargeIntro:bool))]
 }
 FiringOverride {
  entry=Passthrough
  path=AnimGraph.AnimGraphNode_StateMachine_2.FiringOverride
  ids=id
  nodes [
   AnimStateNode_0 = Passthrough() {is=title AnimStateNode}
   AnimStateNode_1 = Fire() {is=title AnimStateNode, bAlwaysResetOnEntry=true}
   AnimStateNode_2 = FireAgain() {is=title AnimStateNode}
   AnimStateTransitionNode_0 = AnimStateTransitionNode() {blendMode=ExpOut, crossfadeDuration=0}
   AnimStateTransitionNode_1 = AnimStateTransitionNode() {bAutomaticRuleBasedOnSequencePlayerInState=true, blendMode=CubicInOut, crossfadeDuration=0.3}
   AnimStateTransitionNode_2 = AnimStateTransitionNode()
   AnimStateTransitionNode_3 = AnimStateTransitionNode()
   AnimStateTransitionNode_5 = AnimStateTransitionNode() {bAutomaticRuleBasedOnSequencePlayerInState=true}
  ]
  links [
   {
    dst=FireAgain
    rule=AnimGraph.AnimGraphNode_StateMachine_2.FiringOverride.AnimStateTransitionNode_2.Transition
    src=Fire
    via=AnimStateTransitionNode_2
   }
   {
    dst=Passthrough
    rule=AnimGraph.AnimGraphNode_StateMachine_2.FiringOverride.AnimStateTransitionNode_1.Transition
    src=Fire
    via=AnimStateTransitionNode_1
   }
   {
    dst=Fire
    rule=AnimGraph.AnimGraphNode_StateMachine_2.FiringOverride.AnimStateTransitionNode_2.Transition
    src=FireAgain
    via=AnimStateTransitionNode_3
   }
   {
    dst=Passthrough
    rule=AnimGraph.AnimGraphNode_StateMachine_2.FiringOverride.AnimStateTransitionNode_5.Transition
    src=FireAgain
    via=AnimStateTransitionNode_5
   }
   {
    dst=Fire
    rule=AnimGraph.AnimGraphNode_StateMachine_2.FiringOverride.AnimStateTransitionNode_0.Transition
    src=Passthrough
    via=AnimStateTransitionNode_0
   }
  ]
 }
 Passthrough {
  path=AnimGraph.AnimGraphNode_StateMachine_2.FiringOverride.AnimStateNode_0.Passthrough
  nodes [Output Animation Pose(Result=Use cached pose 'PostMontage'())]
 }
 Fire {
  path=AnimGraph.AnimGraphNode_StateMachine_2.FiringOverride.AnimStateNode_1.Fire
  nodes [
   n1 = Sequence Player() {
    becomeRelevantFunction {bSelfContext=true, memberGuid=#1, memberName=SetupFireAnim}
    node {bLoopAnimation=false, becomeRelevantFunction {functionName=OnBecomeRelevant}}
   }
   Output Animation Pose(Result=@n1)
  ]
 }
 FireAgain {
  path=AnimGraph.AnimGraphNode_StateMachine_2.FiringOverride.AnimStateNode_2.FireAgain
  nodes [
   n1 = Sequence Player() {
    becomeRelevantFunction {bSelfContext=true, memberGuid=#1, memberName=SetupFireAnim}
    node {bLoopAnimation=false, becomeRelevantFunction {functionName=OnBecomeRelevant}}
   }
   Output Animation Pose(Result=@n1)
  ]
 }
 Transition {
  path=AnimGraph.AnimGraphNode_StateMachine_2.FiringOverride.AnimStateTransitionNode_0.Transition
  nodes [Result(bCanEnterTransition=Get bOnChargeEnd() -> (bOnChargeEnd:bool))]
 }
 Transition {path=AnimGraph.AnimGraphNode_StateMachine_2.FiringOverride.AnimStateTransitionNode_1.Transition, nodes [Result()]}
 Transition {
  path=AnimGraph.AnimGraphNode_StateMachine_2.FiringOverride.AnimStateTransitionNode_2.Transition
  nodes [Result(bCanEnterTransition=Get bDoFireInterrupt() -> (bDoFireInterrupt:bool))]
 }
 Transition {path=AnimGraph.AnimGraphNode_StateMachine_2.FiringOverride.AnimStateTransitionNode_5.Transition, nodes [Result()]}
 EventGraph {links [], nodes []}
 SetupFireAnim {nodes [
   n1 = SetupFireAnim!() -> (then, Context:"Anim Update Context Structure (by ref)", Node:"Anim Node Reference Structure (by ref)") {is=title K2Node_FunctionEntry}
   n2 = Convert to Sequence Player!(Node:"Anim Node Reference Structure (by ref)"=@n1.Node) -> (SequencePlayer:Sequence Player Reference Structure, Result:bool)
   > Branch(Condition=@n2.Result)
   > SetSequenceWithInertialBlending!(
    execute
    UpdateContext:"Anim Update Context Structure (by ref)"=@n1
    SequencePlayer:"Sequence Player Reference Structure (by ref)"=@n2
    Sequence:Anim Sequence Base Object Reference=Select!(
     Option 0:Anim Sequence Base Object Reference=/NanaimoBar/Assets/Animation/Animation/NanaimoBar_Fire_NonTargeted_WEP
     Option 1:Anim Sequence Base Object Reference=&NanaimoBar_Fire_NonTargeted_WEP
     Option 2:Anim Sequence Base Object Reference=/NanaimoBar/Assets/Animation/Animation/NanaimoBar_Fire_Stage02_WEP
     Option 3:Anim Sequence Base Object Reference=/NanaimoBar/Assets/Animation/Animation/NanaimoBar_Fire_Stage03_WEP
     Index:int=Get ChargeStage() -> (ChargeStage=0)
    ) -> (ReturnValue:Anim Sequence Base Object Reference)
    BlendTime:float=0
   ) -> (then, ReturnValue:Sequence Player Reference Structure)
   > Return Node()
  ]}
]
variables []
